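# Physics asset for the dynamic hair bangs on the M_MED_Broccoli medium male player body cosmetic.
# Fifteen small, identical capsule bodies cover the dyn_hair_bangs bones, alongside a larger head capsule.
# Most bangs constrain directly to the head, while bangs 02 and 03 branch from bang 01.
# All constraints share the same swing and tighter twist limit values.
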PhysicsAsset /BRCosmetics/Characters/Player/Male/Medium/Bodies/M_MED_Broccoli/Meshes/Parts/M_MED_Broccoli_Dyn
bodies {
 dyn_hair_bangs_01 [{shapeType=Capsule, length=1, radius=1, center=<0.1 0 0>, rotation=<rot 90 0 0>}]
 dyn_hair_bangs_010 [{shapeType=Capsule, length=1, radius=1, center=<0.1 0 0>, rotation=<rot 90 0 0>}]
 dyn_hair_bangs_011 [{shapeType=Capsule, length=1, radius=1, center=<0.1 0 0>, rotation=<rot 90 0 0>}]
 dyn_hair_bangs_012 [{shapeType=Capsule, length=1, radius=1, center=<0.1 0 0>, rotation=<rot 90 0 0>}]
 dyn_hair_bangs_013 [{shapeType=Capsule, length=1, radius=1, center=<0.1 0 0>, rotation=<rot 90 0 0>}]
 dyn_hair_bangs_014 [{shapeType=Capsule, length=1, radius=1, center=<0.1 0 0>, rotation=<rot 90 0 0>}]
 dyn_hair_bangs_015 [{shapeType=Capsule, length=1, radius=1, center=<0.1 0 0>, rotation=<rot 90 0 0>}]
 dyn_hair_bangs_016 [{shapeType=Capsule, length=1, radius=1, center=<0.1 0 0>, rotation=<rot 90 0 0>}]
 dyn_hair_bangs_02 [{shapeType=Capsule, length=1, radius=1, center=<0.1 0 0>, rotation=<rot 90 0 0>}]
 dyn_hair_bangs_03 [{shapeType=Capsule, length=1, radius=1, center=<0.1 0 0>, rotation=<rot 90 0 0>}]
 dyn_hair_bangs_04 [{shapeType=Capsule, length=1, radius=1, center=<0.1 0 0>, rotation=<rot 90 0 0>}]
 dyn_hair_bangs_05 [{shapeType=Capsule, length=1, radius=1, center=<0.1 0 0>, rotation=<rot 90 0 0>}]
 dyn_hair_bangs_07 [{shapeType=Capsule, length=1, radius=1, center=<0.1 0 0>, rotation=<rot 90 0 0>}]
 dyn_hair_bangs_08 [{shapeType=Capsule, length=1, radius=1, center=<0.1 0 0>, rotation=<rot 90 0 0>}]
 dyn_hair_bangs_09 [{shapeType=Capsule, length=1, radius=1, center=<0.1 0 0>, rotation=<rot 90 0 0>}]
 head [{shapeType=Capsule, length=6.8222, radius=6.9955, rotation=<rot 90 0 0>}]
}
constraints [
 {bone1Name=dyn_hair_bangs_01, bone2Name=head, swing1LimitDegrees=30, swing2LimitDegrees=30, twistLimitDegrees=5}
 {bone1Name=dyn_hair_bangs_02, bone2Name=dyn_hair_bangs_01, swing1LimitDegrees=30, swing2LimitDegrees=30, twistLimitDegrees=5}
 {bone1Name=dyn_hair_bangs_03, bone2Name=dyn_hair_bangs_01, swing1LimitDegrees=30, swing2LimitDegrees=30, twistLimitDegrees=5}
 {bone1Name=dyn_hair_bangs_04, bone2Name=head, swing1LimitDegrees=30, swing2LimitDegrees=30, twistLimitDegrees=5}
 {bone1Name=dyn_hair_bangs_05, bone2Name=head, swing1LimitDegrees=30, swing2LimitDegrees=30, twistLimitDegrees=5}
 {bone1Name=dyn_hair_bangs_07, bone2Name=head, swing1LimitDegrees=30, swing2LimitDegrees=30, twistLimitDegrees=5}
 {bone1Name=dyn_hair_bangs_08, bone2Name=head, swing1LimitDegrees=30, swing2LimitDegrees=30, twistLimitDegrees=5}
 {bone1Name=dyn_hair_bangs_09, bone2Name=head, swing1LimitDegrees=30, swing2LimitDegrees=30, twistLimitDegrees=5}
 {bone1Name=dyn_hair_bangs_010, bone2Name=head, swing1LimitDegrees=30, swing2LimitDegrees=30, twistLimitDegrees=5}
 {bone1Name=dyn_hair_bangs_011, bone2Name=head, swing1LimitDegrees=30, swing2LimitDegrees=30, twistLimitDegrees=5}
 {bone1Name=dyn_hair_bangs_012, bone2Name=head, swing1LimitDegrees=30, swing2LimitDegrees=30, twistLimitDegrees=5}
 {bone1Name=dyn_hair_bangs_013, bone2Name=head, swing1LimitDegrees=30, swing2LimitDegrees=30, twistLimitDegrees=5}
 {bone1Name=dyn_hair_bangs_014, bone2Name=head, swing1LimitDegrees=30, swing2LimitDegrees=30, twistLimitDegrees=5}
 {bone1Name=dyn_hair_bangs_015, bone2Name=head, swing1LimitDegrees=30, swing2LimitDegrees=30, twistLimitDegrees=5}
 {bone1Name=dyn_hair_bangs_016, bone2Name=head, swing1LimitDegrees=30, swing2LimitDegrees=30, twistLimitDegrees=5}
]
